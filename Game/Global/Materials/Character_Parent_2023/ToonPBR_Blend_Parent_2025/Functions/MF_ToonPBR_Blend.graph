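# Character material function that blends PBR attributes with an emissive-driven toon look by a cel-shading amount, behind a static switch.
# The toon color combines banded light-direction shading with noise-broken edges, skin-masked Phong or Blinn highlights, and a curve-atlas metallic gradient.
# Optional stepped-time grunge overlays base color, and the result passes through brightness, LUT and time-of-day adjustment.
# The normal output can blend surface normals toward normals baked in vertex color.
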MaterialFunction /Game/Global/Materials/Character_Parent_2023/ToonPBR_Blend_Parent_2025/Functions/MF_ToonPBR_Blend
graph {nodes [
  Constant()
  n1 = NamedRerouteUsage()
  n2 = Saturate(None=NamedRerouteUsage())
  n3 = Constant()
  n4 = SetMaterialAttributes!(
   MaterialAttributes=@n1
   EmissiveColor=Multiply(A=@n2, B=ScalarParameter() {defaultValue=1, group=ToonPBR Blend, parameterName=Emissive - Scale, sortPriority=3})
   BaseColor=Multiply(A=@n2, B=ScalarParameter() {group=ToonPBR Blend, parameterName=BaseColor - Scale, sortPriority=2})
   Specular=@n3
   Metallic=@n3
   Roughness=Constant() {r=1}
   Opacity=@n3
  ) {attributeSetTypes [
    #1
    #2
    #3
    #4
    #5
    #6
   ]}
  n5 = BlendMaterialAttributes(
   A=SetMaterialAttributes!(MaterialAttributes=@n1, BaseColor=@n2) {attributeSetTypes [#2]}
   B=@n4
   Alpha=Saturate(None=ScalarParameter() {defaultValue=0.2, group=ToonPBR Blend, parameterName=Cel Shading - Blend, sliderMax=1, sortPriority=1})
  )
  n6 = NamedRerouteUsage()
  FunctionOutput(None=StaticSwitchParameter(
    True=SetMaterialAttributes!(MaterialAttributes=@n5, Normal=@n6) {attributeSetTypes [#7]}
    False=NamedRerouteUsage()
   ) {defaultValue=true, group=ToonPBR Blend, parameterName=Use Toon PBR Blending, sortPriority=0}) {outputName=MA_Out}
  MaterialFunctionCall!(Position, UVRadialScale) {materialFunction=/Game/Characters/Player/Common/MaterialFunctions/Superlevel/CharacterStabilizedScreenSpaceUVs}
  n7 = NamedRerouteUsage()
  MaterialFunctionCall!(In=@n7) {materialFunction=/Game/Packages/Fortress_SharedMaterials/Anime_Master_Material/MF_TODCustomColors}
  n8 = MaterialFunctionCall() {materialFunction=/Game/Global/Materials/Character_Parent_2023/Anime_Parent_2023/MaterialFunctions/MFC_AnimeShader_LightDirection}
  n9 = ScalarParameter() {
   defaultValue=0.333
   desc="LockedZ only for the specular highlight.\r\nfor a streched highlight in on the Z axis use the Blinn highlight"
   group=ToonPBR Blend
   parameterName=Specular Highlight - Locked Z Vector
   sliderMax=1
  }
  n10 = AppendVector(A=Constant2Vector() {g=1, r=1}, B=Saturate(None=OneMinus(None=@n9)))
  n11 = Multiply(A=Add(A=CameraVectorWS(), B=@n8.Direction Vector), B=@n10)
  n12 = NamedRerouteUsage()
  n13 = Multiply(A=@n8.Direction Vector, B=@n10)
  n14 = StaticSwitchParameter(
   True=DotProduct(A=Normalize(VectorInput=@n11), B=@n12) {desc=Blinn}
   False=DotProduct(A=Normalize(VectorInput=@n13), B=ReflectionVectorWS(CustomWorldNormal=@n12)) {desc=Phong}
  ) {
   desc="Default is the Phong highlight.\r\nUse Blinn for a vertical streched highlight"
   group=ToonPBR Blend
   parameterName=Specular - Use Blinn Highlight
  }
  n15 = NamedRerouteUsage()
  n16 = ScalarParameter() {group=ToonPBR Blend, parameterName=Specular Highlight - Camera Facing - Blend, sliderMax=1}
  n17 = SphereMask(
   A=Saturate(None=LinearInterpolate(A=@n14, B=@n15, Alpha=@n16))
   B=Constant() {r=1}
   Radius=ScalarParameter() {defaultValue=0.1, group=ToonPBR Blend, parameterName=Specular Highlight - Size, sliderMax=1, sliderMin=0.0001}
   Hardness=ScalarParameter() {defaultValue=0.5, group=ToonPBR Blend, parameterName=Specular Highlight - Hardness, sliderMax=1}
  ) {hardnessPercent=78}
  n18 = VectorParameter() {desc="RGB=Color, A=Opacity", group=ToonPBR Blend, parameterName=Specular Highlight - Color, defaultValue=<rgba 1 1 1 0.25>}
  NamedRerouteDeclaration(None=Multiply(A=Saturate(None=@n17), B=Multiply(A=@n18.RGB, B=@n18.RGB))) {name=Specular_Curve, nodeColor=<rgba 1 0.4453 0 1>}
  n19 = ScalarParameter() {
   desc="Metallic - Color Curve \r\n\r\nThis number allows you to choose which index you want to use in your Curve Atlas so you don't need to use multiple curve atlas for different color curves"
   group=ToonPBR Blend
   parameterName=Metallic - Curve Index
   sortPriority=30
  }
  n20 = ScalarParameter() {
   defaultValue=2
   desc="Metallic - Color Curve \r\n\r\nThis is the total amount of color curves in your Curve Atlas"
   group=ToonPBR Blend
   parameterName=Metallic - Curve Amount
   sortPriority=29
  }
  n21 = NamedRerouteUsage()
  n22 = Constant() {desc=Metallic Band Amount, r=3}
  n23 = Multiply(A=Saturate(None=Power(Base=@n21) {constExponent=2.2}), B=@n22)
  n24 = Multiply(A=Saturate(None=ScalarParameter() {defaultValue=0.75, group=ToonPBR Blend, parameterName=Metallic - Roughness, sliderMax=1})) {constB=2}
  n25 = LinearInterpolate(
   B=Divide(
    A=MaterialFunctionCall!(In=@n23, Ramp Width=@n24) {materialFunction=/Engine/Functions/Engine_MaterialFunctions01/Math/SmoothCeil}.Result
    B=@n22
   )
   Alpha=ScalarParameter() {defaultValue=1, group=ToonPBR Blend, parameterName=Metallic - Intensity, sliderMax=1}
  ) {constA=1}
  n26 = ScalarParameter() {
   defaultValue=-0.3
   desc="The Bias will adjust the Metallic gradient.\r\n0.0 = the center matches the Color\r\n1.0 = the inner ring matches the Color\r\n2.0 = the outer ring match the Color"
   group=ToonPBR Blend
   parameterName=Metallic - Bias
   sliderMax=2
   sliderMin=-1
  }
  n27 = Saturate(None=Add(A=Saturate(None=@n25), B=Multiply(A=@n26) {constB=0.3333}))
  n28 = TextureSampleParameter2D(UVs=Saturate(None=MaterialFunctionCall!(CurveIndex=@n19, CurveNum=@n20, GreyMapIn=@n27) {materialFunction=/Game/Effects/Niagara/MasterMaterials/Functions/Common/FXCommon_CurveAtlasUV}.UV)) {
   group=ToonPBR Blend
   parameterName=Metallic - Curve Atlas
   samplerType=SAMPLERTYPE_LinearColor
   sortPriority=31
   texture=/Game/Global/Textures/Default/Curve/FN_Character_Atlas
  }
  NamedRerouteDeclaration(None=@n28.RGB) {name=Metallic_Gradient, nodeColor=<rgba 0.1107 1 0.7992 1>}
  NamedRerouteDeclaration(None=MaterialFunctionCall() {materialFunction=/Game/Global/Materials/Character_Parent_2023/Anime_Parent_2023/MaterialFunctions/MFC_ActorAlighnedScreenPosition}.UVs) {name=ObjectAlighedScreenPosition, nodeColor=<rgba 1 0.867 0 1>}
  n29 = PixelNormalWS()
  NamedRerouteDeclaration(None=DotProduct(A=NamedRerouteDeclaration(None=@n29) {name=NormalWS, nodeColor=<rgba 0.0078 0 1 1>}, B=CameraVectorWS())) {name=NdotV, nodeColor=<rgba 1 0.6242 0 1>}
  n30 = MaterialFunctionCall() {materialFunction=&MFC_AnimeShader_LightDirection}
  n31 = NamedRerouteUsage()
  n32 = Multiply(A=NamedRerouteDeclaration(None=DotProduct(A=@n30.Direction Vector, B=@n31)) {name=LdotN, nodeColor=<rgba 0.2656 0 1 1>}) {constB=0.5}
  NamedRerouteDeclaration(None=Saturate(None=Add(A=@n32) {constB=0.5})) {name=Lambert, nodeColor=<rgba 0 0.9297 1 1>}
  NamedRerouteDeclaration(None=MaterialFunctionCall!(Fade Offset=Constant() {r=50}, Fade Distance=Constant() {r=400}) {materialFunction=/Game/Global/Materials/Character_Parent_2023/Anime_Parent_2023/MaterialFunctions/MFC_DistanceFade_FOV_compensation}.Result) {name=DistanceBlending, nodeColor=<rgba 0.4766 1 0 1>}
  n33 = ScalarParameter() {defaultValue=0.675, group=ToonPBR Blend, parameterName=Cel Shading - Toon Shadow - Darkening, sliderMax=1, sortPriority=1}
  n34 = StaticSwitchParameter(True=Saturate(None=NamedRerouteUsage()), False=NamedRerouteUsage()) {group=ToonPBR Blend, parameterName=Use Animated Grunge, sortPriority=7}
  n35 = LinearInterpolate(A=Multiply(A=@n33, B=@n34) {constB=0.66}, B=@n34, Alpha=NamedRerouteUsage())
  n36 = LinearInterpolate(A=@n35, B=Multiply(A=@n35, B=NamedRerouteUsage()), Alpha=NamedRerouteUsage())
  n37 = NamedRerouteUsage()
  n38 = Multiply(A=@n37, B=ScalarParameter() {defaultValue=0.05, group=ToonPBR Blend, parameterName=Specular - Intensity, sliderMax=4})
  n39 = Multiply(A=@n37, B=ScalarParameter() {defaultValue=0.35, group=ToonPBR Blend, parameterName=Specular - Skin - Intensity, sliderMax=4})
  n40 = NamedRerouteUsage()
  n41 = ScalarParameter() {defaultValue=1, group=ToonPBR Blend, parameterName=Specular - Metallic - Intensity, sliderMax=4}
  n42 = Multiply(
   A=LinearInterpolate(A=LinearInterpolate(A=@n38, B=@n39, Alpha=@n40), B=Multiply(A=@n37, B=@n41), Alpha=NamedRerouteUsage())
   B=NamedRerouteUsage()
  )
  NamedRerouteDeclaration(None=Saturate(None=Add(A=@n36, B=@n42))) {name=ColorRaw, nodeColor=<rgba 0 0.5781 1 1>}
  n43 = SetMaterialAttributes!(MaterialAttributes, BaseColor) {attributeSetTypes [#2]}
  n44 = GetMaterialAttributes(None=NamedRerouteDeclaration(None=FunctionInput(Preview=@n43) {inputName=MA_In, inputType=FunctionInput_MaterialAttributes}) {name=MA_In, nodeColor=<rgba 1 0.9609 0 1>}) {attributeGetTypes [
    #2
    #7
    #4
    #3
   ]}
  NamedRerouteDeclaration(None=@n44.Metallic) {name=Metallic_In, nodeColor=<rgba 0 0.9581 1 1>}
  NamedRerouteDeclaration(None=@n44.Specular) {name=Specular_In, nodeColor=<rgba 1 0.9289 0 1>}
  n45 = ChannelMaskParameter(None=FunctionInput() {bUsePreviewValueAsDefault=true, inputName=M, sortPriority=1}) {group=ToonPBR Blend, maskChannel=Blue, parameterName=SkinMask_Channel, sortPriority=6, defaultValue=<rgba 0 0 1 0>}
  NamedRerouteDeclaration(None=@n45.R) {name=SkinMask, nodeColor=<rgba 0.9219 0 1 1>}
  n46 = ComponentMask(None=ViewProperty() {property=MEVP_ResolutionFraction}.Property) {r=true}
  NamedRerouteDeclaration(None=Saturate(None=InverseLinearInterpolate(Value=@n46) {constA=0.3})) {name=ResScale, nodeColor=<rgba 0 0.6484 1 1>}
  n47 = Add(
   A=NamedRerouteUsage()
   B=MaterialFunctionCall() {materialFunction=&/MF_SteppedTime}.TimeStep
  )
  n48 = ScalarParameter() {defaultValue=0.2, group=ToonPBR Blend, parameterName=Animated Grunge - Tiling, sliderMax=2, sortPriority=9}
  n49 = TextureSampleParameter2D(UVs=Frac(None=Multiply(A=@n47, B=@n48))) {
   group=ToonPBR Blend
   parameterName=Animated Grunge - Texture
   samplerType=SAMPLERTYPE_Masks
   sortPriority=8
   texture=&T_WhiteMask
  }
  n50 = ChannelMaskParameter(None=@n49.RGB) {group=ToonPBR Blend, maskChannel=Green, parameterName=Animated Grunge - Channel, sortPriority=9, defaultValue=<rgba 0 1 0 0>}
  n51 = ScalarParameter() {defaultValue=0.2, group=ToonPBR Blend, parameterName=Animated Grunge - Bias, sliderMax=1, sliderMin=-1, sortPriority=10}
  n52 = Add(A=@n50.R, B=@n51)
  n53 = ScalarParameter() {defaultValue=0.8, group=ToonPBR Blend, parameterName=Animated Grunge - Exponent, sliderMax=2, sortPriority=11}
  NamedRerouteDeclaration(None=MaterialFunctionCall(Base=NamedRerouteUsage(), Blend=Power(Base=@n52, Exp=@n53) {constExponent=0.4545}) {materialFunction=&Blend_Overlay}.Result) {name=BaseColor_withGrunge, nodeColor=<rgba 1 0.7969 0 1>}
  NamedRerouteDeclaration(None=@n44.BaseColor) {name=BaseColor_In, nodeColor=<rgba 1 0.1346 0 1>}
  n54 = ScalarParameter() {defaultValue=0.75, group=ToonPBR Blend, parameterName=Shadow - Edge Breakup - Min, sliderMax=1}
  n55 = ScalarParameter() {defaultValue=1, group=ToonPBR Blend, parameterName=Shadow - Edge Breakup - Max, sliderMax=1}
  n56 = MaterialFunctionCall() {materialFunction=&MFC_AnimeShader_LightDirection}
  n57 = PixelNormalWS()
  n58 = Saturate(None=Add(
    A=DotProduct(A=@n56.Direction Vector, B=@n57)
    B=ScalarParameter() {group=ToonPBR Blend, parameterName=Shading Curve - Bias, sliderMax=1, sliderMin=-1}
   ))
  n59 = ScalarParameter() {defaultValue=0.6, group=ToonPBR Blend, parameterName=Shading Curve - Width, sliderMax=1}
  n60 = ScalarParameter() {defaultValue=1, group=ToonPBR Blend, parameterName=Shading Curve - Bands, sliderMax=4, sliderMin=1}
  n61 = VectorParameter() {group=ToonPBR Blend, parameterName="Shadow - Edge Breakup  - Tiling", defaultValue=<rgba 16 16 0 1>}
  n62 = TextureObjectParameter() {
   group=ToonPBR Blend
   parameterName=Shadow - Edge Breakup - Texture
   samplerType=SAMPLERTYPE_LinearGrayscale
   texture=/Engine/EngineMaterials/Good64x64TilingNoiseHighFreq
  }
  n63 = TextureSample(UVs=Multiply(A=TextureCoordinate(), B=ComponentMask(None=@n61.RGB) {g=true, r=true}), Tex=@n62) {samplerType=SAMPLERTYPE_Masks, texture=&T_WhiteMask}
  n64 = ScalarParameter() {defaultValue=1, group=ToonPBR Blend, parameterName=Shadow - Edge Breakup - Intensity, sliderMax=1}
  n65 = MaterialFunctionCall!(
   Base=Saturate(None=MaterialFunctionCall!(NdL=@n58, Width=@n59, Bands Amount=@n60) {materialFunction=/Game/Global/Materials/Character_Parent_2023/Anime_Parent_2023/MaterialFunctions/MF_SmoothCelShadingWithBands}.Divide)
   Add=Multiply(A=NamedRerouteDeclaration(None=@n63.R) {name=StippingMask, nodeColor=<rgba 0.2891 0 1 1>}, B=Saturate(None=@n64))
  ) {materialFunction=/Engine/Functions/Engine_MaterialFunctions02/Math/MultiplyAdd}
  NamedRerouteDeclaration(None=Saturate(None=SmoothStep(Min=@n54, Max=@n55, Value=@n65.Result))) {name=Shading_Gradient, nodeColor=<rgba 1 0.1875 0 1>}
  n66 = TextureObject() {samplerType=SAMPLERTYPE_Masks, texture=&T_WhiteMask}
  n67 = Add(A=VertexInterpolator(VS=LocalPosition() {localOrigin=InstancePreSkinning}.XYZ).PS, B=Constant3Vector() {constant=<rgba 0 0 14.6837 1>})
  n68 = Constant() {r=35}
  NamedRerouteDeclaration(None=ComponentMask(None=MaterialFunctionCall!(
     ColorTex=@n66
     NormalTex
     Pos=@n67
     Nrm
     ColorUVScale=@n68
     ColorBlendFactor
     NormalStrength
     NormalUVScale
     NormalBlendFactor
     Local Space Inputs
    ) {materialFunction=/Game/Global/MaterialFunctions/MF_BiPlanar}.Color) {b=true}) {name=Grunge, nodeColor=<rgba 0.5469 0 1 1>}
  NamedRerouteDeclaration(None=@n44.Normal) {name=Normal_In, nodeColor=<rgba 0 0.3061 1 1>}
  n69 = ScalarParameter() {defaultValue=1, group=ToonPBR Blend, parameterName=Color - Brightness - High, sortPriority=5}
  n70 = ScalarParameter() {defaultValue=2, group=ToonPBR Blend, parameterName=Color - Brightness - Low, sortPriority=4}
  n71 = Constant() {r=1}
  n72 = Saturate(None=LinearInterpolate(A=Multiply(A=@n7, B=@n69), B=Multiply(A=@n7, B=@n70), Alpha=Desaturation(None=@n7, Fraction=@n71)))
  NamedRerouteDeclaration(None=MaterialFunctionCall!(Color=MaterialFunctionCall!(In=@n72, LUT, LUT Multiplier) {materialFunction=/Game/Packages/Fortress_SharedMaterials/Anime_Master_Material/MF_ApplyLUT}.Result) {materialFunction=/Game/Packages/Fortress_SharedMaterials/Anime_Master_Material/MF_TODColorAdjustment}.Result) {name=ColorFinal, nodeColor=<rgba 1 0.6797 0 1>}
  n73 = VertexColor()
  n74 = Multiply(A=@n73) {constB=2}
  n75 = ScalarParameter() {desc="0.0 = surface normals, 1.0 baked vertex color normal", group=ToonPBR Blend, parameterName=Vertex Baked Normal - Blend, sliderMax=1}
  n76 = LinearInterpolate(
   A=Transform(None=NamedRerouteUsage())
   B=Transform(None=Subtract(A=@n74)) {transformSourceType=TRANSFORMSOURCE_Local}
   Alpha=Multiply(A=@n73.A, B=@n75)
  )
  NamedRerouteDeclaration(None=Transform(None=Normalize(VectorInput=@n76)) {transformSourceType=TRANSFORMSOURCE_World, transformType=TRANSFORM_Tangent}) {name=Normal_Out, nodeColor=<rgba 0 1 0.7995 1>}
  NamedRerouteUsage()
  NamedRerouteUsage()
  Saturate(None=Abs(None=OneMinus(None=@n33)))
  n77 = ScalarParameter() {defaultValue=0.025, group=ToonPBR Blend, parameterName=GrungeMapIntensity}
  n78 = ScalarParameter() {defaultValue=-0.3, group=ToonPBR Blend, parameterName=GrungeMapIntensity_Skin}
  n79 = Desaturation(None=NamedRerouteUsage(), Fraction=Constant() {r=1})
  Subtract(B=Multiply(A=NamedRerouteUsage(), B=LinearInterpolate(A=@n77, B=@n78, Alpha=@n79)))
  Transform(None=NamedRerouteUsage())
  VertexNormalWS()
 ]}
