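# Parent material for skeletal characters that, by its name, blends toon and PBR looks.
# Chains three functions on material attributes: character PreFX builds the base, MF_ToonPBR_Blend processes it, and character PostFX produces the final result.
# PreFX also feeds PostFX directly for its mask, position, distance-blend and Fresnel inputs.
# Shading model comes from the graph, covering lit, subsurface, skin profile, clear coat and cloth, with mesh deformer support.
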Material /Game/Global/Materials/Character_Parent_2023/ToonPBR_Blend_Parent_2025/M_ToonPBR_Blend_Parent
graph {
 properties {
  bAutomaticallySetUsageInEditor=false
  bUseMaterialAttributes=true
  bUsedWithMeshDeformer=true
  bUsedWithSkeletalMesh=true
  shadingModel=MSM_FromMaterialExpression
  subsurfaceProfile=&SS_HeroSkin_02
  usedShadingModels=Default Lit | Subsurface | Clear Coat | Subsurface Profile | Cloth
 }
 nodes [
  n1 = MaterialFunctionCall() {materialFunction=&MF_CharacterParent_2023_PreFX}
  MaterialFunctionCall(
   Material=MaterialFunctionCall!(MA_In=@n1.Result, M=@n1.Result) {materialFunction=/Game/Global/Materials/Character_Parent_2023/ToonPBR_Blend_Parent_2025/Functions/MF_ToonPBR_Blend}.MA_Out
   M=@n1.M
   Pre Skinned Local Position=@n1.M
   Distance Blend=@n1.M
   Fresnel=@n1.M
  ) {materialFunction=&MF_CharacterParent_2023_PostFX}
 ]
}
